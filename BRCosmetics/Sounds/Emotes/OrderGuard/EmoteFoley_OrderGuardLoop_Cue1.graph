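# Looping foley SoundCue for the Order Guard emote, routed through the emote foley sound class and property sheet.
# A looping wave player feeds an Enveloper that holds full volume at first, then eases it down modestly as playback continues.
# Emote attenuation and concurrency rules apply, the cue keeps playing when silent, and it has a send to the music audio-analysis submix.
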SoundCue /BRCosmetics/Sounds/Emotes/OrderGuard/EmoteFoley_OrderGuardLoop_Cue1
root=Enveloper_0
nodes [
 {
  id=Enveloper_0
  children [WavePlayer_0]
  properties {pitchCurve {editorCurveData {keys [{time=0}]}}, volumeCurve {editorCurveData {keys [<at 0 1>,<at 15 1>,<at 20 0.77>]}}}
 }
 {id=WavePlayer_0, properties {bLooping=true, soundWaveAssetPtr=&/Emote_Order_Guard_Loop}}
]
properties {
 attenuationSettings=&Fort_Emotes_Athena_Att
 audioPropertiesSheet=/BRCosmetics/Sounds/Settings/PropertySheets/Emotes/APS_Cosmetic_Emote_Foley
 maxDistance=1600
 soundClassObject=/Game/Sounds/FortSoundClassesAndModes/Fort_Foley_Emote
 virtualizationMode=PlayWhenSilent
 volumeMultiplier=1.3
 concurrencyOverrides {maxCount=3}
 concurrencySet [/Game/Sounds/Concurrency_Rules/Athena_EmoteFoley_Conc]
 soundSubmixSends [{minSendLevel=0, soundSubmix=&Fortnite_AudioAnalysis_Music}]
}
